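; Major spire dummy actor for the spire boss that plunges to a target location, then dies and explodes.
; It relays the target from a non-replicated State Tree through a replicated variable, whose OnRep starts projectile movement with scaled gravity.
; A looping timer compares height against the target instead of using collision, for cheaper long-range visuals.
; On arrival it stops simulation, executes the explosion GameplayCue and broadcasts OnFlowEnded.
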(event EventBeginPlay)

(event Custom|PlungeOnServerAndClients
  (params (Target :type Vector))
  ; Using the Actor as an intermediary to replicate a target location and plunge towards that as the State Tree controlling this flow does not have replication
  (Variables|Default|SetTargetPlungeLocation Target))

(event Custom|TriggerPlunge
  (bind _projectilemovement (Variables|Default|GetProjectileMovement))
  ; Using the Actor as an intermediary to replicate a target location and plunge towards that as the State Tree controlling this flow does not have replication
  (Variables|/Script/Engine.ProjectileMovementComponent|SetbSimulationEnabled true _projectilemovement)
  (Variables|/Script/Engine.ProjectileMovementComponent|SetProjectileGravityScale 3.0 _projectilemovement) ; Using Gravity Scale for now to control \nfalling speed. If needed we can use a\ntimeline-like approach but want to keep it \nsimple if we don't end up needing that
  (bind _returnvalue (Utilities|Time|SetTimerbyEvent (EventDispatchers|CreateEvent :!event Custom|PlungeUpdate) (Variables|Default|GetUpdateRate) true true)) ; Slower tick exclusively for checking target location\nDon't want to use collision to stop simulating to \nminimize perf cost as this will be visible from very far away
  (Variables|Default|SetPlungeUpdateTimerHandle _returnvalue))

(event Custom|PlungeUpdate
  ; Using the Actor as an intermediary to replicate a target location and plunge towards that as the State Tree controlling this flow does not have replication
  (if (CallFunction|IsBelowTarget)
    (Game|Components|ProjectileMovement|StopSimulating (Variables|Default|GetProjectileMovement) (Collision|MakeHitResult true))
    (Utilities|Time|ClearandInvalidateTimerbyHandle (Variables|Default|GetPlungeUpdateTimerHandle))
    (CallFunction|OnPlungeEnded)))

(event Custom|OnPlungeEnded
  ; Using the Actor as an intermediary to replicate a target location and plunge towards that as the State Tree controlling this flow does not have replication
  (Gameplay|Die)
  (GameplayCue|ExecuteGameplayCue self (Variables|Default|GetGC_Explosion) (Class|AthenaFunctionLibrary|MakeSimpleFortEffectContext self self))
  (Default|CallOnFlowEnded))

(fn IsBelowTarget
  (bind (_returnvalue_x _returnvalue_y _returnvalue_z) (Math|Vector|BreakVector (Transformation|GetActorLocation)))
  (bind (_targetplungelocation_x _targetplungelocation_y _targetplungelocation_z) (Math|Vector|BreakVector (Variables|GetTargetPlungeLocation_X)))
  (return (<= (- _returnvalue_z _targetplungelocation_z) 0)))

(fn OnRep_TargetPlungeLocation
  (CallFunction|TriggerPlunge))

(fn ConstructionScript)
